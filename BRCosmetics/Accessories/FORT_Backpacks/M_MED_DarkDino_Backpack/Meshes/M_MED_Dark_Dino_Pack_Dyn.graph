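# Physics asset for the Dark Dino backpack cosmetic, defining simulation bodies for its dynamic spine and tail bones.
# Box bodies cover two spine bones and three tail segments, with a small capsule on the pack root bone.
# Both spines and the first tail segment constrain directly to the pack, while later tail segments chain to one another.
# Tail joints allow wider swing and twist than the tighter, asymmetrically limited spine joints.
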PhysicsAsset /BRCosmetics/Accessories/FORT_Backpacks/M_MED_DarkDino_Backpack/Meshes/M_MED_Dark_Dino_Pack_Dyn
bodies {
 dyn_spine_01 [{shapeType=Box, extentX=7.5, extentY=5, extentZ=7.5, center=<3 0 0>}]
 dyn_spine_02 [{shapeType=Box, extentX=7.5, extentY=5, extentZ=16, center=<3 0 0>}]
 dyn_tail_01 [{shapeType=Box, extentX=11, extentY=9, extentZ=7.5, center=<4.6 0 0>}]
 dyn_tail_02 [{shapeType=Box, extentX=7.8, extentY=6.6, extentZ=5.5, center=<4.2 0 0>}]
 dyn_tail_03 [{shapeType=Box, extentX=7.5, extentY=4.5, extentZ=5.4, center=<3.4 0 0>}]
 pack [{shapeType=Capsule, length=2, radius=2}]
}
constraints [
 {bone1Name=dyn_tail_01, bone2Name=pack, swing1LimitDegrees=45, swing2LimitDegrees=45, twistLimitDegrees=10}
 {bone1Name=dyn_spine_01, bone2Name=pack, swing1LimitDegrees=30, swing2LimitDegrees=10, twistLimitDegrees=5}
 {bone1Name=dyn_spine_02, bone2Name=pack, swing1LimitDegrees=30, swing2LimitDegrees=10, twistLimitDegrees=5}
 {bone1Name=dyn_tail_02, bone2Name=dyn_tail_01, swing1LimitDegrees=45, swing2LimitDegrees=45, twistLimitDegrees=10}
 {bone1Name=dyn_tail_03, bone2Name=dyn_tail_02, swing1LimitDegrees=45, swing2LimitDegrees=45, twistLimitDegrees=10}
]
